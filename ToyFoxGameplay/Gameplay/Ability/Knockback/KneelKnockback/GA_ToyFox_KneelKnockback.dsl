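; GA_ToyFox_KneelKnockback is the ToyFox knockback gameplay ability tied to a kneel, judging by its name, built on a parent knockback ability.
; On activation it runs the parent logic, then makes the avatar player pawn ignore collision with the giant pawn, fetched from the ToyFox game state component, for the knockback duration.
; It overrides FindLauncher to return the effect causer from the granting gameplay effect's context as the launcher.
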(event Ability|EventActivateAbility
  (|CallParentFunction)
  (Collision|AddIgnoredActor (CallFunction|GetAvatarasFortPlayerPawn) (Giant|GetGiantPawn (Actor|GetComponentbyClass (Game|GetGameState) "/ToyFoxGameplay/Gameplay/B_ToyFoxGameStateComponent.B_ToyFoxGameStateComponent_C")) (ScalableFloat|GetCachedScalableFloatValue (Variables|Default|GetKnockbackDuration))))

(event Ability|EventOnEndAbility
  (params (bWasCancelled :type Boolean)))

(fn FindLauncher
  (params (DefaultLauncher :type Actor%20Object%20Reference))
  (return (Ability|EffectContext|GetEffectCauser (Ability|GetGrantedbyEffectContext))))
